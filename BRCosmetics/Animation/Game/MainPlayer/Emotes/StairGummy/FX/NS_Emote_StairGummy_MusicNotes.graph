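# Niagara system for the StairGummy emote that emits music note sprites over a wavy ribbon trail, both drifting along negative X and upward.
# MusicTrail001 scales its ribbon spawn rate by User.SpawnCurve, widens with age and writes a sine-offset Wave Pos for undulation.
# NE_MusicNotes spawns SubUV sprites that pop in, rock on a sine rotation and shrink out.
# A PopBurst emitter keyed to the system's DeathSpawn and DeathPosition values is present but disabled.
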NiagaraSystem /BRCosmetics/Animation/Game/MainPlayer/Emotes/StairGummy/FX/NS_Emote_StairGummy_MusicNotes
dependencies {
 usedDataInterfaces [
  &NiagaraDataInterfaceArrayColor
  &NiagaraDataInterfaceCurve
  &NiagaraDataInterfaceVector2DCurve
  &NiagaraDataInterfaceColorCurve
  &NiagaraDataInterfaceArrayPosition
 ]
 usedDynamicInputs []
 usedModules [
  &self:SystemSpawnScript.SystemScriptSource.SystemScriptGraph.NiagaraNodeAssignment_0.SetVariables_#1
  &SystemState
  &BlendSpaceHelper
  &self:ReadDeath
  &self:SystemSpawnScript.SystemScriptSource.SystemScriptGraph.NiagaraNodeAssignment_1.SetVariables_#2
 ]
}
emitters {
 MusicTrail001 {
  emitterUpdateScript [
   &EmitterState()
   &SpawnRate(SpawnRate=&Multiply_Float(A=$User.SpawnCurve, B=60), Use Spawn Probability=false)
  ]
  particleSpawnScript [
   &InitializeParticle(
    Color Mode=::NewEnumerator0
    Lifetime=1.8
    Position Offset=<20 0 0>
    Ribbon Width=12
    Ribbon Width Mode=::NewEnumerator1
    Sprite Size=<70 25>
    Sprite Size Mode=::NewEnumerator4
    UsePositionOffset=true
    Write Material Random=true
   )
   &AddVelocity(Velocity=&RandomRangeVector(Minimum=<-115 0 0>, Maximum=<-115 0 0>))
  ]
  particleUpdateScript [
   &ParticleState()
   &GravityForce(Gravity=<0 0 70>)
   &SolveForcesAndVelocity(Clamp Velocity=false, Limit Acceleration=false)
   &ScaleColor(
    Linear Color Curve=DataInterface() {propertyValues {
      AlphaCurve {keys [
        <at 0 0 RCIM_Cubic>
        {arriveTangent=0.0715, interpMode=RCIM_Cubic, leaveTangent=0.0715, tangentMode=RCTM_User, time=0.2769, value=0.7}
        {arriveTangent=-0.1489, interpMode=RCIM_Cubic, leaveTangent=-0.1489, tangentMode=RCTM_User, time=0.8472, value=0.7}
        <at 1 0 RCIM_Cubic>
       ]}
      BlueCurve {keys [<at 0 0>]}
      GreenCurve {keys [<at 0 0>]}
      RedCurve {keys [<at 0 0>]}
     }}
    Scale Mode=::NewEnumerator2
   )
   &ScaleRibbonWidth(Ribbon Width Scale=&FloatFromCurve(FloatCurve=DataInterface() {propertyValues {Curve {keys [{interpMode=RCIM_Cubic, leaveTangentWeight=2.3909, tangentMode=RCTM_User, time=0, value=1}, <at 1 3 RCIM_Cubic>]}}}))
   /SetVariables(Particles.Wave Pos:NiagaraPosition=&Add_Vector(
     A=Linked() {linkedVariable {name=Particles.Position, type=NiagaraPosition}}
     B=&MakeVector(
      X=&Sine(
       Normalized Angle=&Add_Float(A=$Emitter.Age, B=&Multiply_Float(A=$Particles.NormalizedAge, B=-0.5))
       Period=3
       Scale=&FloatFromCurve(
        FloatCurve=DataInterface() {propertyValues {Curve {keys [<at 0 0>, {arriveTangent=1, interpMode=RCIM_Cubic, leaveTangent=1, time=0.2304, value=0.7182}, <at 1 1>]}}}
        Scale Curve=28
       )
       Bias=0
      )
      Y=0
      Z=0
     )
    )) {isSetParameters=true}
  ]
  rendererClasses [&NiagaraRibbonRendererProperties]
  renderers [{rendererClass=&NiagaraRibbonRendererProperties}]
 }
 NE_MusicNotes {
  emitterUpdateScript [
   &EmitterState()
   &SpawnRate(SpawnRate=3, Use Spawn Probability=true)
  ]
  particleSpawnScript [
   /Niagara/Modules/Update/SubUV/SubUVAnimation(SubUV Animation Mode=::NewEnumerator1)
   &DynamicMaterialParameters(
    Index 0 Param 1=&Multiply_Float(A=$System.Age, B=2)
    Index 0 Param 2=&RandomRangeFloat(Minimum=4, Maximum=6, Evaluation Type=::NewEnumerator1)
    Index 0 Param 3=1
    Index 0 Param 4=1
    Module.Param0WriteEnabled=true
    Module.Param1WriteEnabled=true
    Module.Param2WriteEnabled=false
    Module.Param3WriteEnabled=false
   )
   &InitializeParticle(
    Color=<rgba 0 0 0 1>
    Lifetime=&RandomRangeFloat(Minimum=2, Maximum=3.2)
    Mass Mode=::NewEnumerator1
    Position Offset=<0 0 1>
    Sprite Rotation Angle=&RandomRangeFloat(Minimum=-0.1, Maximum=0.1)
    Sprite Rotation Mode=::NewEnumerator1
    Sprite Size=&Vector2DFromFloat(Value=15)
    Sprite Size Mode=::NewEnumerator4
    UsePositionOffset=true
    Write Material Random=true
   )
   &AddVelocity(
    Scale Added Velocity=&VectorFromFloat(Value=1.5)
    Velocity=&RandomRangeVector(Minimum=<-25 0 0>, Maximum=<-40 1 0>)
   )
   &Color(Color=&SelectLinearColorFromArray(
     Array Sampling Mode=::NewEnumerator1
     Direct Array Index=&RandomRangeInt(Minimum=0, Maximum=3)
     Color Selection Array=DataInterface() {propertyValues {ColorData [<rgba 0 0 0 1>, <rgba 0 0 0 1>, <rgba 0 0 0 1>, <rgba 0 0 0 1>]}}
     Random Seed=?
    ))
  ]
  particleUpdateScript [
   &ParticleState()
   &ScaleColor(
    Scale Alpha=&FloatFromCurve(FloatCurve=DataInterface() {propertyValues {Curve {keys [<at 0.9 1 RCIM_Cubic>,<at 1 0 RCIM_Cubic>]}}}, Scale Curve=0.75)
    Scale RGB=&VectorFromFloat(Value=&FloatFromCurve(FloatCurve=DataInterface() {propertyValues {Curve {keys [
          <at 0 0 RCIM_Cubic>
          {interpMode=RCIM_Cubic, tangentMode=RCTM_User, time=0.1062, value=6}
          {arriveTangent=-7.4341, interpMode=RCIM_Cubic, tangentMode=RCTM_Break, time=0.2671, value=0.9767}
          <at 1 1 RCIM_Cubic>
         ]}}}))
    ScaleA=true
    ScaleRGB=false
   )
   &GravityForce(Gravity=<0 0 20>)
   &SolveForcesAndVelocity(Clamp Velocity=false, Limit Acceleration=false)
   &ScaleSpriteSize(Scale Factor=&Vector2DFromFloat(Value=&FloatFromCurve(FloatCurve=DataInterface() {propertyValues {Curve {keys [
          <at 0 0 RCIM_Cubic>
          {interpMode=RCIM_Cubic, tangentMode=RCTM_User, time=0.18, value=1.2189}
          {arriveTangent=-0.3562, interpMode=RCIM_Cubic, leaveTangent=-0.3562, tangentMode=RCTM_User, time=0.9, value=1.2189}
          <at 1 0 RCIM_Cubic>
         ]}}})))
   /SetVariables(
    Particles.LightEnabled:NiagaraBool=&SetBoolByFloatComparison(A=$Particles.NormalizedAge, B=0.08, Comparison Type=::NewEnumerator7)
    Particles.SpriteRotation:NiagaraFloat=&Sine(
     Normalized Angle=&Add_Float(A=$Emitter.Age, B=&RandomRangeFloat(Minimum=0, Maximum=1))
     Period=1
     Scale=20
     Bias=0
    )
   ) {isSetParameters=true}
  ]
  rendererClasses [&NiagaraSpriteRendererProperties]
  renderers [{rendererClass=&NiagaraSpriteRendererProperties}]
 }
 PopBurst {
  bEnabled=false
  emitterUpdateScript [
   &EmitterState()
   &SpawnPerFrame(Spawn:NiagaraBool=$System.DeathSpawn, Spawn Count=8, Use Spawn Probability=false) {module=SpawnPerFrame001}
  ]
  particleSpawnScript [
   &PartitionParticles(Partition Count Mode:ENiagara_InfiniteLoopDuration=::NewEnumerator1, Partition Mode:ENiagara_PartitionMode_v2=::NewEnumerator1)
   &InitializeParticle(
    Color=&SelectLinearColorFromArray(
     Array Sampling Mode=::NewEnumerator1
     Direct Array Index=$Particles.PartitionParticles.Partition
     Direct Array Mode=::NewEnumerator1
     Color Selection Array=DataInterface() {propertyValues {ColorData [
        <rgba 0.8802 0.4371 0 1>
        <rgba 0.1009 0.4463 1 1>
        <rgba 1 0.0447 0 1>
        <rgba 0.3356 0.776 0.0243 1>
        <rgba 1 0.125 0.6934 1>
        <rgba 0.1665 0.004 0.8229 1>
       ]}}
     Random Seed=?
    )
    Lifetime Max=0.18
    Lifetime Min=0.08
    Lifetime Mode=::NewEnumerator1
    Mesh Scale Mode=::NewEnumerator3
    Mesh Uniform Scale=0.005
    Position=&TransformPosition(Position=$System.DeathPosition, Source Space=::Local, Destination Space=::World)
    Position Mode=::NewEnumerator1
    Sprite Size Max=<1 15>
    Sprite Size Min=<0.4 9>
    Sprite Size Mode=::NewEnumerator2
    UsePositionOffset=false
    Write Material Random=true
   ) {enabled=false}
   &InitializeParticle(
    Color=&SelectLinearColorFromArray(
     Array Sampling Mode=::NewEnumerator1
     Direct Array Index=$Particles.PartitionParticles.Partition
     Direct Array Mode=::NewEnumerator1
     Color Selection Array=DataInterface() {propertyValues {ColorData [<rgba 0 0 0 0.7304>]}}
     Random Seed=?
    )
    Lifetime Max=0.18
    Lifetime Min=0.08
    Lifetime Mode=::NewEnumerator1
    Mesh Scale Mode=::NewEnumerator3
    Mesh Uniform Scale=0.005
    Position=&TransformPosition(Position=$System.DeathPosition, Source Space=::Local, Destination Space=::World)
    Position Mode=::NewEnumerator1
    Sprite Size Max=<1 15>
    Sprite Size Min=<0.4 9>
    Sprite Size Mode=::NewEnumerator2
    UsePositionOffset=false
    Write Material Random=true
   ) {module=InitializeParticle001}
   &ShapeLocation(
    MaskSpawnGroup=false
    Offset Mode=::NewEnumerator0
    Override Seed=false
    Shape Origin=$Particles.Position
    Sphere Radius=1
    Sphere Surface Distribution=1
    UseOwnerScale=true
   ) {enabled=false}
   &AddVelocity(
    Module.Use Falloff Distance=false
    Override Seed=false
    Velocity Mode=::NewEnumerator1
    Velocity Origin=$Particles.Position
    Velocity Speed=200
   )
   &CameraOffset(Camera Offset Amount=-10) {enabled=false}
  ]
  particleUpdateScript [
   &Drag(Drag=&RandomRangeFloat(Minimum=6, Maximum=8), Use Linear Drag=true, Use Rotational Drag=false)
   &InheritSourceMovement(
    Applied Position Scale=&MakeCustomVectorFromBool(
     Bool=&SetBoolByFloatComparison(A=&VectorLength(Vector=$Engine.Owner.Velocity), B=500)
     True Vector=<1 1 1>
     False Vector=<0.1 0.1 0>
    )
    ApplyForce=false
    ApplyPosition=true
    ApplyPositionFalloff=false
   )
   &SolveForcesAndVelocity(Clamp Velocity=false, Limit Acceleration=false)
   &ScaleColor(
    Scale Alpha=&FloatFromCurve(
     FloatCurve=DataInterface() {propertyValues {Curve {
        defaultValue=0.3
        keys [
         {tangentMode=RCTM_User, time=0, value=0.5}
         {arriveTangent=-0.0405, interpMode=RCIM_Cubic, leaveTangent=-0.0405, tangentMode=RCTM_User, time=0.7009, value=0.5}
         {arriveTangent=-4.9566, interpMode=RCIM_Cubic, leaveTangent=-4.9566, tangentMode=RCTM_User, time=1, value=0}
        ]
       }}}
     Scale Curve=&FNTimeOfDayScalar(Night Scale=1, Morning Scale=1, Evening Scale=1)
    )
    Scale RGB=&MakeVectorFromLinearColorRGB(Color=&FNTimeOfDayColor(
      Night Color=<rgba 0.3609 0.4715 0.5469 1>
      Morning Color=<rgba 0.651 0.4517 0.4205 1>
      Day Color=<rgba 1 1 1 1>
      Evening Color=<rgba 0.651 0.4517 0.4205 1>
     ))
    ScaleA=true
    ScaleRGB=true
   )
   &ParticleState()
   &DynamicMaterialParameters(
    Index 0 Param 2=1
    Index 0 Param 3=&FloatFromCurve(
     FloatCurve=DataInterface() {propertyValues {Curve {keys [
         {arriveTangent=3.0114, interpMode=RCIM_Cubic, leaveTangent=3.0114, tangentMode=RCTM_User, time=0, value=0}
         {arriveTangent=0.9456, interpMode=RCIM_Cubic, leaveTangent=0.9456, tangentMode=RCTM_User, time=0.4343, value=0.7445}
         {arriveTangent=0.3236, interpMode=RCIM_Cubic, leaveTangent=0.7406, tangentMode=RCTM_User, time=1, value=1}
        ]}}}
     Scale Curve=0.2
    )
    Index 0 Param 4=1
    Module.Index0Float0Write=true
    Module.Index0Float1Write=true
    Module.Index0Float2Write=true
    Module.Index0Float3Write=true
   ) {enabled=false}
   &ScaleSpriteSize(
    Scale Factor=&MakeVector2D(
     X=&FloatFromCurve(FloatCurve=DataInterface() {propertyValues {Curve {keys [<at 0 1>,<at 0.4954 1>,<at 1 1>]}}})
     Y=&FloatFromCurve(FloatCurve=DataInterface() {propertyValues {Curve {keys [<at 0 0>,<at 0.15 1>,<at 0.4 1>,<at 1 0>]}}})
    )
    Scale Sprite Size Mode=::NewEnumerator2
   )
  ]
  rendererClasses [&NiagaraSpriteRendererProperties, &NiagaraMeshRendererProperties]
  renderers [
   {rendererClass=&NiagaraSpriteRendererProperties}
   {rendererClass=&NiagaraMeshRendererProperties, rendererIndex=1}
  ]
 }
}
summary {
 systemName=NS_Emote_StairGummy_MusicNotes
 emitters [
  {
   bEnabled=false
   emitterName=PopBurst
   rendererClasses [&NiagaraSpriteRendererProperties, &NiagaraMeshRendererProperties]
  }
  {emitterName=MusicTrail001, rendererClasses [&NiagaraRibbonRendererProperties]}
  {emitterName=NE_MusicNotes, rendererClasses [&NiagaraSpriteRendererProperties]}
 ]
 userVariables [
  User.bIsFrontEnd {type=NiagaraBool}
  User.bIsFrontEndPreview {type=NiagaraBool}
  User.OwnerLocalVelocity {type=Vector3f, description="", defaultValue {struct=Vector3f, value=<0 0 0>}}
  User.SpawnCurve {type=NiagaraFloat, defaultValue=1, description=""}
 ]
}
