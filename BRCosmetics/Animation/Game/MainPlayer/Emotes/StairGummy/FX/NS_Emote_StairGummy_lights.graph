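# Niagara effect for the StairGummy emote that puts blinking lights and beams on the stair gadget prop.
# Three sprite emitters and a mesh beam emitter share spawn rate, lifetime and camera offset parameters, with user-set colors and alpha fade curves.
# Each particle's UniqueID times User.id is partitioned to pick one of the gadget's 22 sockets, distributing flashes without per-light emitters.
# GroundLightMesh001 separately bursts a long-lived ground glow mesh spanning the 12-second loop.
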NiagaraSystem /BRCosmetics/Animation/Game/MainPlayer/Emotes/StairGummy/FX/NS_Emote_StairGummy_lights
dependencies {
 usedDataInterfaces [
  &NiagaraDataInterfaceArrayFloat2
  &NiagaraDataInterfaceSkeletalMesh
  &NiagaraDataInterfaceColorCurve
  &NiagaraDataInterfaceCurve
 ]
 usedDynamicInputs [&MakeFloatFromVector2d, &RandomRangeVector2D]
 usedModules [
  &SystemState
  &self:SystemSpawnScript.SystemScriptSource.SystemScriptGraph.NiagaraNodeAssignment_2.SetVariables_#1
 ]
}
emitters {
 BottomLights {
  emitterUpdateScript [
   &EmitterState(Inactive Response=::NewEnumerator1, Life Cycle Mode=::NewEnumerator1, Loop Duration=12, UseLoopDelay=false)
   &SpawnRate(SpawnRate=$User.SpawnRate, Use Spawn Probability=false) {module=SpawnRate001}
  ]
  particleSpawnScript [
   &PartitionParticles(
    ID to Partition:NiagaraInt32=&Multiply_Int(A=$Particles.UniqueID, B=&MakeIntFromFloat(Float=$User.id))
    Number Of Partitions=&MakeIntFromFloat(Float=$User.number of partions)
   )
   &InitializeParticle(
    Color Mode=::NewEnumerator0
    Lifetime=$User.Lifetime
    Sprite Size Mode=::NewEnumerator3
    Uniform Sprite Size=15
    UsePositionOffset=true
    Write Material Random=true
   )
   &ApplyInitialForces()
   &SkeletalMeshLocation(
    Bone / Socket Index=$Particles.PartitionParticles.Partition
    MaskSpawnGroup=false
    Mesh Sampling Type=::NewEnumerator5
    Sampled Orientation Coordinate Space=::Local
    Sampled Position Coordinate Space=::Local
    Sampled Position Offset=<0 0 2>
    Sampled Position Offset Coordinate Space:ENiagaraSkelMeshTransforms=::NewEnumerator2
    Sampled Velocity Coordinate Space:ENiagaraCoordinateSpace=::Local
    Skeletal Mesh=DataInterface() {propertyValues {
      DefaultMesh=null
      ExcludeBoneName=root
      OptionalComponentClass=null
      PreviewMesh=/BRCosmetics/Animation/Props/Gadgets_StairGummy/Meshes/StairGummy_Gadget
      SkinningMode=SkinOnTheFly
      SoftSourceActor=null
      SourceMode=Default
      UvSetIndex=0
      WholeMeshLOD=-1
      bExcludeBone=false
      bReadDeformedGeometry=true
      bRequireCurrentFrameData=true
      ComponentTags []
      FilteredBones []
      FilteredSockets [
       base_JNTSocket_0
       base_JNTSocket_1
       base_JNTSocket_2
       base_JNTSocket_3
       base_JNTSocket_4
       base_JNTSocket_5
       base_JNTSocket_6
       base_JNTSocket_7
       base_JNTSocket_8
       base_JNTSocket_9
       base_JNTSocket_10
       base_JNTSocket_11
       base_JNTSocket_12
       base_JNTSocket_13
       base_JNTSocket_14
       base_JNTSocket_15
       base_JNTSocket_16
       base_JNTSocket_17
       base_JNTSocket_18
       base_JNTSocket_19
       base_JNTSocket_20
       base_JNTSocket_21
      ]
      MeshUserParameter {parameter {name=None, typeDefHandle {registeredTypeIndex=1}, varData []}}
      SamplingRegions []
     }}
    Socket Sampling Mode=::NewEnumerator5
   )
   &CameraOffset(Camera Offset Amount=$User.CamOffset)
  ]
  particleUpdateScript [
   &ParticleState()
   &DynamicMaterialParameters(
    Index 0 Param 1=0.15
    Module.Index0Param0Write=true
    Module.Index0Param1Write=true
    Module.Index0Param2Write=true
    Module.Index0Param3Write=true
   )
   /SetVariables(
    Particles.GlowColor=&MakeLinearColorFromVectorAndFloat(
     "Vector (RGB)"=&Multiply_Vector(
      A=&MakeVectorFromLinearColorRGB(Color=&ColorFromCurve(ColorCurve=DataInterface() {propertyValues {AlphaCurve {keys []}, BlueCurve {keys [<at 0.1028 1>]}, GreenCurve {keys [<at 0.1028 0.7913>]}, RedCurve {keys [<at 0.1028 0.5647>]}}}))
      B=&VectorFromFloat(Value=1)
     )
     "Float (Alpha)"=0.5
    )
    Particles.GlowScale:Vector2f=&Vector2DFromFloat(Value=55)
   ) {isSetParameters=true, module=SetVariables_#2}
   &ScaleColor(
    Scale Alpha=&FloatFromCurve(
     FloatCurve=DataInterface() {propertyValues {Curve {keys [
         <at 0 0 RCIM_Cubic>
         {arriveTangent=0.1809, interpMode=RCIM_Cubic, leaveTangent=0.1809, tangentMode=RCTM_User, time=0.25, value=1}
         {arriveTangent=-0.2147, interpMode=RCIM_Cubic, leaveTangent=-0.2147, tangentMode=RCTM_User, time=0.75, value=1}
         <at 1 0 RCIM_Cubic>
        ]}}}
     Scale Curve=0.8
    )
    Scale RGB=&MakeVectorFromLinearColorRGB(Color=&ColorFromCurve(ColorCurve=DataInterface() {propertyValues {
        AlphaCurve {keys []}
        BlueCurve {keys [<at 0.1028 0.8376>,<at 0.1302 0.8879>,<at 0.2711 0.8879>,<at 0.296 0.8376>,<at 0.5191 0.8376>,<at 0.5465 0.8879>,<at 0.7559 0.8879>,<at 0.7808 0.8376>]}
        GreenCurve {keys [<at 0.1028 0.15>,<at 0.1302 0.4332>,<at 0.2711 0.4332>,<at 0.296 0.15>,<at 0.5191 0.15>,<at 0.5465 0.4332>,<at 0.7559 0.4332>,<at 0.7808 0.15>]}
        RedCurve {keys [<at 0.1028 1>,<at 0.1302 0>,<at 0.2711 0>,<at 0.296 1>,<at 0.5191 1>,<at 0.5465 0>,<at 0.7559 0>,<at 0.7808 1>]}
       }}))
    ScaleA=true
    ScaleRGB=false
   )
   &ScaleColor(Color Scale:LinearColor=$User.bottomcolor, Scale Mode=3) {module=ScaleColor003}
   &ScaleColor(
    Scale Alpha=&MakeFloatFromVector(
     Vector=&MakeVectorFromLinearColorRGB(Color=&FN_Select_Color_from_Time_Of_Day(Morning=<rgba 1 0.8463 0.7189 1>, Day=<rgba 1 1 1 1>, Evening=<rgba 1 0.8371 0.7936 1>, Night=<rgba 1 1 1 1>))
     Channel=::NewEnumerator0
    )
    Scale RGB=&MakeVectorFromLinearColorRGB(Color=&FN_Select_Color_from_Time_Of_Day(Morning=<rgba 0.567 0.793 1 1>, Day=<rgba 0.567 0.793 1 1>, Evening=<rgba 0.567 0.793 1 1>, Night=<rgba 0.8505 1.1895 1.5 1>))
    ScaleA=true
    ScaleRGB=true
   ) {enabled=false, module=ScaleColor004}
  ]
  rendererClasses [&NiagaraSpriteRendererProperties]
  renderers [{rendererClass=&NiagaraSpriteRendererProperties}]
 }
 Flashing_Lights {
  emitterUpdateScript [
   &EmitterState(Life Cycle Mode=::NewEnumerator1, Loop Delay=0.9, Loop Duration=12, UseLoopDelay=false)
   &SpawnRate(SpawnRate=$User.SpawnRate, Use Spawn Probability=false)
  ]
  particleSpawnScript [
   &PartitionParticles(
    ID to Partition:NiagaraInt32=&Multiply_Int(A=$Particles.UniqueID, B=&MakeIntFromFloat(Float=$User.id))
    Number Of Partitions=&MakeIntFromFloat(Float=$User.number of partions)
   )
   &InitializeParticle(
    Color=&Multiply_LinearColorByFloat(Linear Color=$User.bottomcolor, Float=1.25)
    Lifetime=$User.Lifetime
    Position Offset=<0 0 -60>
    Sprite Size=&SelectVector2DFromArray(
     Direct Array Index=$Particles.PartitionParticles.Partition
     Vector 2D Selection Array=DataInterface() {propertyValues {FloatData [
        <11 25>
        <11 25>
        <11 25>
        <11 25>
        <11 25>
        <11 25>
        <11 25>
        <11 25>
        <11 25>
        <11 25>
        <11 25>
        <11 25>
        <11 25>
        <11 25>
        <11 25>
        <11 25>
        <11 25>
        <11 25>
        <11 25>
        <11 25>
        <11 25>
        <11 25>
       ]}}
    )
    Sprite Size Mode=::NewEnumerator4
    UsePositionOffset=true
    Write Material Random=true
   )
   &AddVelocity(Velocity=<0 0 1>, Velocity Speed Scale=0.01)
   &ApplyInitialForces()
   &SkeletalMeshLocation(
    Bone / Socket Index=$Particles.PartitionParticles.Partition
    MaskSpawnGroup=false
    Mesh Sampling Type=::NewEnumerator5
    Sampled Orientation Coordinate Space=::Local
    Sampled Position Coordinate Space=::Local
    Sampled Position Offset=<0 0 -6>
    Sampled Position Offset Coordinate Space:ENiagaraSkelMeshTransforms=::NewEnumerator2
    Sampled Velocity Coordinate Space:ENiagaraCoordinateSpace=::Local
    Skeletal Mesh=DataInterface() {propertyValues {
      DefaultMesh=null
      ExcludeBoneName=root
      OptionalComponentClass=null
      PreviewMesh=&StairGummy_Gadget
      SkinningMode=SkinOnTheFly
      SoftSourceActor=null
      SourceMode=Default
      UvSetIndex=0
      WholeMeshLOD=-1
      bExcludeBone=false
      bReadDeformedGeometry=true
      bRequireCurrentFrameData=true
      ComponentTags []
      FilteredBones []
      FilteredSockets [
       base_JNTSocket_0
       base_JNTSocket_1
       base_JNTSocket_2
       base_JNTSocket_3
       base_JNTSocket_4
       base_JNTSocket_5
       base_JNTSocket_6
       base_JNTSocket_7
       base_JNTSocket_8
       base_JNTSocket_9
       base_JNTSocket_10
       base_JNTSocket_11
       base_JNTSocket_12
       base_JNTSocket_13
       base_JNTSocket_14
       base_JNTSocket_15
       base_JNTSocket_16
       base_JNTSocket_17
       base_JNTSocket_18
       base_JNTSocket_19
       base_JNTSocket_20
       base_JNTSocket_21
      ]
      MeshUserParameter {parameter {name=None, typeDefHandle {registeredTypeIndex=1}, varData []}}
      SamplingRegions []
     }}
    Socket Sampling Mode=::NewEnumerator5
   )
   &CameraOffset(Camera Offset Amount=$User.CamOffset)
  ]
  particleUpdateScript [
   &ParticleState()
   &SolveForcesAndVelocity(Clamp Velocity=false, Limit Acceleration=false)
   &ScaleColor(Scale Alpha=0.75, ScaleA=true, ScaleRGB=false)
   &ScaleColor(
    Scale Alpha=&FloatFromCurve(
     FloatCurve=DataInterface() {propertyValues {Curve {keys [
         <at 0 0 RCIM_Cubic>
         {arriveTangent=0.1809, interpMode=RCIM_Cubic, leaveTangent=0.1809, tangentMode=RCTM_User, time=0.25, value=1}
         {arriveTangent=-0.2147, interpMode=RCIM_Cubic, leaveTangent=-0.2147, tangentMode=RCTM_User, time=0.75, value=1}
         <at 1 0 RCIM_Cubic>
        ]}}}
     Scale Curve=0.8
    )
    Scale RGB=&MakeVectorFromLinearColorRGB(Color=&ColorFromCurve(ColorCurve=DataInterface() {propertyValues {
        AlphaCurve {keys []}
        BlueCurve {keys [<at 0.1028 0.8376>,<at 0.1302 0.8879>,<at 0.2711 0.8879>,<at 0.296 0.8376>,<at 0.5191 0.8376>,<at 0.5465 0.8879>,<at 0.7559 0.8879>,<at 0.7808 0.8376>]}
        GreenCurve {keys [<at 0.1028 0.15>,<at 0.1302 0.4332>,<at 0.2711 0.4332>,<at 0.296 0.15>,<at 0.5191 0.15>,<at 0.5465 0.4332>,<at 0.7559 0.4332>,<at 0.7808 0.15>]}
        RedCurve {keys [<at 0.1028 1>,<at 0.1302 0>,<at 0.2711 0>,<at 0.296 1>,<at 0.5191 1>,<at 0.5465 0>,<at 0.7559 0>,<at 0.7808 1>]}
       }}))
    ScaleA=true
    ScaleRGB=false
   ) {module=ScaleColor001}
  ]
  rendererClasses [&NiagaraSpriteRendererProperties]
  renderers [{rendererClass=&NiagaraSpriteRendererProperties}]
 }
 Flat_Lights {
  emitterUpdateScript [
   &EmitterState(Inactive Response=::NewEnumerator1, Life Cycle Mode=::NewEnumerator1, Loop Duration=12, UseLoopDelay=false)
   &SpawnRate(SpawnRate=$User.SpawnRate, Use Spawn Probability=false)
  ]
  particleSpawnScript [
   &PartitionParticles(
    ID to Partition:NiagaraInt32=&Multiply_Int(A=$Particles.UniqueID, B=&MakeIntFromFloat(Float=$User.id))
    Number Of Partitions=&MakeIntFromFloat(Float=$User.number of partions)
   )
   &InitializeParticle(
    Color Mode=::NewEnumerator0
    Lifetime=$User.Lifetime
    Sprite Size Mode=::NewEnumerator3
    Uniform Sprite Size=15
    UsePositionOffset=true
    Write Material Random=true
   )
   &SkeletalMeshLocation(
    Bone / Socket Index=$Particles.PartitionParticles.Partition
    MaskSpawnGroup=false
    Mesh Sampling Type=::NewEnumerator5
    Sampled Orientation Coordinate Space=::Local
    Sampled Position Coordinate Space=::Local
    Sampled Position Offset=<0 0 1>
    Sampled Position Offset Coordinate Space:ENiagaraSkelMeshTransforms=::NewEnumerator2
    Sampled Velocity Coordinate Space:ENiagaraCoordinateSpace=::Local
    Skeletal Mesh=DataInterface() {propertyValues {
      DefaultMesh=null
      ExcludeBoneName=root
      OptionalComponentClass=null
      PreviewMesh=&StairGummy_Gadget
      SkinningMode=SkinOnTheFly
      SoftSourceActor=null
      SourceMode=Default
      UvSetIndex=0
      WholeMeshLOD=-1
      bExcludeBone=false
      bReadDeformedGeometry=true
      bRequireCurrentFrameData=true
      ComponentTags []
      FilteredBones []
      FilteredSockets [
       base_JNTSocket_0
       base_JNTSocket_1
       base_JNTSocket_2
       base_JNTSocket_3
       base_JNTSocket_4
       base_JNTSocket_5
       base_JNTSocket_6
       base_JNTSocket_7
       base_JNTSocket_8
       base_JNTSocket_9
       base_JNTSocket_10
       base_JNTSocket_11
       base_JNTSocket_12
       base_JNTSocket_13
       base_JNTSocket_14
       base_JNTSocket_15
       base_JNTSocket_16
       base_JNTSocket_17
       base_JNTSocket_18
       base_JNTSocket_19
       base_JNTSocket_20
       base_JNTSocket_21
      ]
      MeshUserParameter {parameter {name=None, typeDefHandle {registeredTypeIndex=1}, varData []}}
      SamplingRegions []
     }}
    Socket Sampling Mode=::NewEnumerator5
   )
   &ApplyInitialForces()
   &AlignSpriteToMeshOrientation(Mesh Orientation Relative Sprite Facing Vector:Vector3f=<0 0 1>, SetAlignment=true, SetFacing=true)
   &CameraOffset(Camera Offset Amount=$User.CamOffset)
  ]
  particleUpdateScript [
   &ParticleState()
   &DynamicMaterialParameters(
    Index 0 Param 1=0.15
    Module.Index0Param0Write=true
    Module.Index0Param1Write=true
    Module.Index0Param2Write=true
    Module.Index0Param3Write=true
   )
   /SetVariables(Particles.DiskScale:Vector3f=&Multiply_VectorByFloat(Vector=<1 1 0.5>, Float=0.7)) {isSetParameters=true}
   /SetVariables(
    Particles.GlowColor=&MakeLinearColorFromVectorAndFloat(
     "Vector (RGB)"=&Multiply_Vector(
      A=&MakeVectorFromLinearColorRGB(Color=&ColorFromCurve(ColorCurve=DataInterface() {propertyValues {AlphaCurve {keys []}, BlueCurve {keys [<at 0.1028 1>]}, GreenCurve {keys [<at 0.1028 0.7913>]}, RedCurve {keys [<at 0.1028 0.5647>]}}}))
      B=&VectorFromFloat(Value=1)
     )
     "Float (Alpha)"=0.5
    )
    Particles.GlowScale:Vector2f=&Vector2DFromFloat(Value=55)
   ) {isSetParameters=true, module=SetVariables_#2}
   &ScaleColor(Color Scale:LinearColor=$User.bottomcolor, Scale Mode=3) {module=ScaleColor003}
   &ScaleColor(
    Scale Alpha=&MakeFloatFromVector(
     Vector=&MakeVectorFromLinearColorRGB(Color=&FN_Select_Color_from_Time_Of_Day(Morning=<rgba 1 0.8463 0.7189 1>, Day=<rgba 1 1 1 1>, Evening=<rgba 1 0.8371 0.7936 1>, Night=<rgba 1 1 1 1>))
     Channel=::NewEnumerator0
    )
    Scale RGB=&MakeVectorFromLinearColorRGB(Color=&FN_Select_Color_from_Time_Of_Day(Morning=<rgba 0.567 0.793 1 1>, Day=<rgba 0.567 0.793 1 1>, Evening=<rgba 0.567 0.793 1 1>, Night=<rgba 0.8505 1.1895 1.5 1>))
    ScaleA=true
    ScaleRGB=true
   ) {enabled=false, module=ScaleColor004}
   &ScaleColor(
    Scale Alpha=&FloatFromCurve(
     FloatCurve=DataInterface() {propertyValues {Curve {keys [
         <at 0 0 RCIM_Cubic>
         {arriveTangent=0.1809, interpMode=RCIM_Cubic, leaveTangent=0.1809, tangentMode=RCTM_User, time=0.25, value=1}
         {arriveTangent=-0.2147, interpMode=RCIM_Cubic, leaveTangent=-0.2147, tangentMode=RCTM_User, time=0.75, value=1}
         <at 1 0 RCIM_Cubic>
        ]}}}
     Scale Curve=0.8
    )
    Scale RGB=&MakeVectorFromLinearColorRGB(Color=&ColorFromCurve(ColorCurve=DataInterface() {propertyValues {
        AlphaCurve {keys []}
        BlueCurve {keys [<at 0.1028 0.8376>,<at 0.1302 0.8879>,<at 0.2711 0.8879>,<at 0.296 0.8376>,<at 0.5191 0.8376>,<at 0.5465 0.8879>,<at 0.7559 0.8879>,<at 0.7808 0.8376>]}
        GreenCurve {keys [<at 0.1028 0.15>,<at 0.1302 0.4332>,<at 0.2711 0.4332>,<at 0.296 0.15>,<at 0.5191 0.15>,<at 0.5465 0.4332>,<at 0.7559 0.4332>,<at 0.7808 0.15>]}
        RedCurve {keys [<at 0.1028 1>,<at 0.1302 0>,<at 0.2711 0>,<at 0.296 1>,<at 0.5191 1>,<at 0.5465 0>,<at 0.7559 0>,<at 0.7808 1>]}
       }}))
    ScaleA=true
    ScaleRGB=false
   ) {module=ScaleColor001}
  ]
  rendererClasses [&NiagaraSpriteRendererProperties]
  renderers [{rendererClass=&NiagaraSpriteRendererProperties}]
 }
 GroundLightMesh001 {
  emitterUpdateScript [
   &EmitterState(Inactive Response=::NewEnumerator1, Life Cycle Mode=::NewEnumerator1, Loop Duration=12, UseLoopDelay=false)
   &SpawnBurst_Instantaneous(Use Spawn Probability=false, UseLoopCountLimit=false)
  ]
  particleSpawnScript [
   &FN_SetSpawnGroup(Number of Spawn Groups=7)
   &InitializeParticle(
    Color=<rgba 0.3429 0.1385 0.35 1>
    Lifetime=12.1
    Mesh Scale=<0.75 0.75 10>
    Mesh Scale Mode=::NewEnumerator4
    Position Offset=<0 -25 6>
    UsePositionOffset=true
    Write Material Random=true
   )
   &DynamicMaterialParameters(
    Index 0 Param 1=1
    Index 0 Param 2=0.2
    Module.Index0Float0Write=true
    Module.Index0Float1Write=true
    Module.Index0Float2Write=true
    Module.Index0Float3Write=true
   )
  ]
  particleUpdateScript [
   &ParticleState()
   &ScaleColor(
    Scale Alpha=&Multiply_Float(
     A=&FloatFromCurve(FloatCurve=DataInterface() {propertyValues {Curve {keys [
          {arriveTangent=-2.5, interpMode=RCIM_Cubic, time=0, value=0}
          {arriveTangent=0.0424, interpMode=RCIM_Cubic, leaveTangent=0.0424, tangentMode=RCTM_User, time=0.01, value=1}
          {arriveTangent=-0.0535, interpMode=RCIM_Cubic, leaveTangent=-0.0535, tangentMode=RCTM_User, time=0.99, value=1}
          <at 1 0 RCIM_Cubic>
         ]}}})
     B=0.6
    )
    ScaleA=true
    ScaleRGB=false
   )
   &ScaleMeshSize(Scale Factor=&VectorFromFloat(Value=0.6))
   &SolveForcesAndVelocity(Clamp Velocity=false, Limit Acceleration=false)
  ]
  rendererClasses [&NiagaraMeshRendererProperties]
  renderers [{rendererClass=&NiagaraMeshRendererProperties, rendererIndex=0}]
 }
 beams {
  emitterUpdateScript [
   &EmitterState(Inactive Response=::NewEnumerator1, Life Cycle Mode=::NewEnumerator1, Loop Duration=12, UseLoopDelay=false)
   &SpawnRate(SpawnRate=$User.SpawnRate, Use Spawn Probability=false)
  ]
  particleSpawnScript [
   &PartitionParticles(
    ID to Partition:NiagaraInt32=&Multiply_Int(A=$Particles.UniqueID, B=&MakeIntFromFloat(Float=$User.id))
    Number Of Partitions=&MakeIntFromFloat(Float=$User.number of partions)
   )
   &InitializeParticle(
    Color Mode=::NewEnumerator0
    Lifetime=$User.Lifetime
    Mesh Scale=&SelectVectorFromArray(
     Direct Array Index=$Particles.PartitionParticles.Partition
     Vector Selection Array=DataInterface() {propertyValues {FloatData [
        <0.15 0.15 0.21>
        <0.15 0.15 0.21>
        <0.15 0.15 0.21>
        <0.15 0.15 0.21>
        <0.15 0.15 0.21>
        <0.15 0.15 0.21>
        <0.15 0.15 0.21>
        <0.15 0.15 0.21>
        <0.15 0.15 0.21>
        <0.15 0.15 0.21>
        <0.15 0.15 0.21>
        <0.15 0.15 0.21>
        <0.15 0.15 0.21>
        <0.15 0.15 0.21>
        <0.15 0.15 0.21>
        <0.15 0.15 0.21>
        <0.15 0.15 0.21>
        <0.15 0.15 0.21>
        <0.15 0.15 0.21>
        <0.15 0.15 0.21>
        <0.15 0.15 0.21>
        <0.15 0.15 0.21>
       ]}}
    )
    Mesh Scale Mode=::NewEnumerator4
    Sprite Size Mode=::NewEnumerator3
    Uniform Sprite Size=100
    UsePositionOffset=false
    Write Material Random=true
   )
   &InitialMeshOrientation(
    Mesh Orientation Mode=::NewEnumerator2
    Rotation=&SelectVectorFromArray(
     Direct Array Index=$Particles.PartitionParticles.Partition
     Vector Selection Array=DataInterface() {propertyValues {FloatData [
        <0.5 0 0>
        <0.5 0 0>
        <0.5 0 0>
        <0.5 0 0>
        <0.5 0 0>
        <0.5 0 0>
        <0.5 0 0>
        <0.5 0 0>
        <0.5 0 0>
        <0.5 0 0>
        <0.5 0 0>
        <0.5 0 0>
        <0.5 0 0>
        <0.5 0 0>
       ]}}
    )
    Use Rotation Vector=true
   )
   &SkeletalMeshLocation(
    Bone / Socket Index=$Particles.PartitionParticles.Partition
    MaskSpawnGroup=false
    Mesh Sampling Type=::NewEnumerator5
    Sampled Orientation Coordinate Space=::Local
    Sampled Position Coordinate Space=::Local
    Sampled Position Offset Coordinate Space:ENiagaraSkelMeshTransforms=::NewEnumerator2
    Sampled Velocity Coordinate Space:ENiagaraCoordinateSpace=::Local
    Skeletal Mesh=DataInterface() {propertyValues {
      DefaultMesh=null
      ExcludeBoneName=root
      OptionalComponentClass=null
      PreviewMesh=&StairGummy_Gadget
      SkinningMode=SkinOnTheFly
      SoftSourceActor=null
      SourceMode=Default
      UvSetIndex=0
      WholeMeshLOD=-1
      bExcludeBone=false
      bReadDeformedGeometry=true
      bRequireCurrentFrameData=true
      ComponentTags []
      FilteredBones []
      FilteredSockets [
       base_JNTSocket_0
       base_JNTSocket_1
       base_JNTSocket_2
       base_JNTSocket_3
       base_JNTSocket_4
       base_JNTSocket_5
       base_JNTSocket_6
       base_JNTSocket_7
       base_JNTSocket_8
       base_JNTSocket_9
       base_JNTSocket_10
       base_JNTSocket_11
       base_JNTSocket_12
       base_JNTSocket_13
       base_JNTSocket_14
       base_JNTSocket_15
       base_JNTSocket_16
       base_JNTSocket_17
       base_JNTSocket_18
       base_JNTSocket_19
       base_JNTSocket_20
       base_JNTSocket_21
      ]
      MeshUserParameter {parameter {name=None, typeDefHandle {registeredTypeIndex=1}, varData []}}
      SamplingRegions []
     }}
    Socket Sampling Mode=::NewEnumerator5
   )
   &CameraOffset(Camera Offset Amount=$User.CamOffset)
  ]
  particleUpdateScript [
   &ParticleState()
   &DynamicMaterialParameters(
    Index 0 Param 1=0.15
    Module.Index0Param0Write=true
    Module.Index0Param1Write=true
    Module.Index0Param2Write=true
    Module.Index0Param3Write=true
   )
   /SetVariables(
    Particles.GlowColor=&MakeLinearColorFromVectorAndFloat(
     "Vector (RGB)"=&Multiply_Vector(
      A=&MakeVectorFromLinearColorRGB(Color=&ColorFromCurve(ColorCurve=DataInterface() {propertyValues {AlphaCurve {keys []}, BlueCurve {keys [<at 0.1028 1>]}, GreenCurve {keys [<at 0.1028 0.7913>]}, RedCurve {keys [<at 0.1028 0.5647>]}}}))
      B=&VectorFromFloat(Value=1)
     )
     "Float (Alpha)"=0.5
    )
    Particles.GlowScale:Vector2f=&Vector2DFromFloat(Value=55)
   ) {isSetParameters=true, module=SetVariables_#2}
   &ScaleColor(Color Scale:LinearColor=$User.spotlight color, Scale Mode=3) {module=ScaleColor003}
   &ScaleColor(
    Scale Alpha=&FloatFromCurve(
     FloatCurve=DataInterface() {propertyValues {Curve {keys [
         <at 0 0 RCIM_Cubic>
         {arriveTangent=0.1809, interpMode=RCIM_Cubic, leaveTangent=0.1809, tangentMode=RCTM_User, time=0.25, value=1}
         {arriveTangent=-0.2147, interpMode=RCIM_Cubic, leaveTangent=-0.2147, tangentMode=RCTM_User, time=0.75, value=1}
         <at 1 0 RCIM_Cubic>
        ]}}}
     Scale Curve=0.8
    )
    Scale RGB=&MakeVectorFromLinearColorRGB(Color=&ColorFromCurve(ColorCurve=DataInterface() {propertyValues {
        AlphaCurve {keys []}
        BlueCurve {keys [<at 0.1028 0.8376>,<at 0.1302 0.8879>,<at 0.2711 0.8879>,<at 0.296 0.8376>,<at 0.5191 0.8376>,<at 0.5465 0.8879>,<at 0.7559 0.8879>,<at 0.7808 0.8376>]}
        GreenCurve {keys [<at 0.1028 0.15>,<at 0.1302 0.4332>,<at 0.2711 0.4332>,<at 0.296 0.15>,<at 0.5191 0.15>,<at 0.5465 0.4332>,<at 0.7559 0.4332>,<at 0.7808 0.15>]}
        RedCurve {keys [<at 0.1028 1>,<at 0.1302 0>,<at 0.2711 0>,<at 0.296 1>,<at 0.5191 1>,<at 0.5465 0>,<at 0.7559 0>,<at 0.7808 1>]}
       }}))
    ScaleA=true
    ScaleRGB=false
   ) {module=ScaleColor001}
  ]
  rendererClasses [&NiagaraMeshRendererProperties]
  renderers [{rendererClass=&NiagaraMeshRendererProperties, rendererIndex=0}]
 }
}
summary {
 systemName=NS_Emote_StairGummy_lights
 emitters [
  {emitterName=Flashing_Lights, rendererClasses [&NiagaraSpriteRendererProperties]}
  {emitterName=GroundLightMesh001, rendererClasses [&NiagaraMeshRendererProperties]}
  {emitterName=beams, rendererClasses [&NiagaraMeshRendererProperties]}
  {emitterName=Flat_Lights, rendererClasses [&NiagaraSpriteRendererProperties]}
  {emitterName=BottomLights, rendererClasses [&NiagaraSpriteRendererProperties]}
 ]
 userVariables [
  User.bIsFrontEnd {type=NiagaraBool}
  User.bIsFrontEndPreview {type=NiagaraBool}
  User.bottomcolor {type=LinearColor, description="", defaultValue {struct=LinearColor, value=<rgba 4 3.1951 0 1>}}
  User.CamOffset {type=NiagaraFloat, defaultValue=3, description=""}
  User.id {type=NiagaraFloat, defaultValue=31, description=""}
  User.Lifetime {type=NiagaraFloat, defaultValue=0.8, description=""}
  User.location_Spotlight {
   type=NiagaraDataInterfaceArrayFloat3
   description=""
   defaultValue {
    struct=NiagaraExt_VariableValue_DataInterface
    value {
     dataInterface=&self:NiagaraDataInterfaceArrayFloat3_0
     dataInterfaceClass=&NiagaraDataInterfaceArrayFloat3
    }
   }
  }
  User.number of partions {type=NiagaraFloat, defaultValue=22, description=""}
  User.SpawnRate {type=NiagaraFloat, defaultValue=24, description=""}
  User.spotlight color {type=LinearColor, description="", defaultValue {struct=LinearColor, value=<rgba 0.7 0.6013 0 0.8>}}
  User.Vector {type=Vector3f, description="", defaultValue {struct=Vector3f, value=<0 3 0>}}
 ]
}
